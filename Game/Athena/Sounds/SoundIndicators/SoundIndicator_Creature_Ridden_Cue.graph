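# Silent placeholder Sound Cue in the Athena sound indicator set, named for ridden creatures.
# It has no root node or sound nodes, so it produces no audio and carries only distance settings.
# An attenuation override defines a zero inner radius with a long falloff, which presumably sets the range of the on-screen sound indicator.
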SoundCue /Game/Athena/Sounds/SoundIndicators/SoundIndicator_Creature_Ridden_Cue
root=null
nodes []
properties {bOverrideAttenuation=true, maxDistance=3200, attenuationOverrides {falloffDistance=8000, attenuationShapeExtents {x=0}}}
