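# Sound cue in the Athena sound indicator set, named for unridden creatures, with no nodes or root, so it plays no audio itself.
# It carries only playback settings: an attenuation override with a zero-extent inner shape and a falloff range, giving it a limited audible distance.
# Its concurrency override limits a few instances per owner and stops the oldest when that limit is exceeded.
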SoundCue /Game/Athena/Sounds/SoundIndicators/SoundIndicator_Creature_Unridden_Cue
root=null
nodes []
properties {
 bOverrideAttenuation=true
 maxDistance=3584
 attenuationOverrides {falloffDistance=2500, attenuationShapeExtents {x=0}}
 concurrencyOverrides {bLimitToOwner=true, maxCount=4, resolutionRule=StopOldest}
}
